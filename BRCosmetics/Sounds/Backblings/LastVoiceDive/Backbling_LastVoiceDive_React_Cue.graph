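# Sound cue for the LastVoiceDive back bling's reactive sound, playing box open or box drop effects.
# After a fixed half-second delay, a switch on the StatValue integer parameter selects the branch.
# The first branch plays a box-open wave, while the other three share an evenly weighted random pick of two box-drop waves.
# Playback is quiet and short-range, using close back bling attenuation and the reactive back bling property sheet.
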SoundCue /BRCosmetics/Sounds/Backblings/LastVoiceDive/Backbling_LastVoiceDive_React_Cue
root=Delay_0
nodes [
 {id=Delay_0, children [Switch_2], properties {delayMax=0.5, delayMin=0.5}}
 {id=Random_1, children [WavePlayer_5,WavePlayer_4], properties {weights [1,1]}}
 {id=Switch_2, children [WavePlayer_6,Random_1,Random_1,Random_1], properties {intParameterName=StatValue}}
 {id=WavePlayer_4, properties {soundWaveAssetPtr=&/Backpack_Last_Voice_Dive_BoxDrop_02}}
 {id=WavePlayer_5, properties {soundWaveAssetPtr=&/Backpack_Last_Voice_Dive_BoxDrop_01}}
 {id=WavePlayer_6, properties {soundWaveAssetPtr=&/Backbling_LastVoiceDive_BoxOpen_03}}
]
properties {
 attenuationSettings=/BRCosmetics/Sounds/Other/Attenuation/Fort_BackBling_Close_Att
 audioPropertiesSheet=/BRCosmetics/Sounds/Settings/PropertySheets/Reactive/APS_Cosmetic_Reactive_Backbling
 maxDistance=1111.11
 soundClassObject=/Game/Sounds/FortSoundClassesAndModes/Fort_Backblings
 volumeMultiplier=0.15
}
